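# Physics asset for the SwissKale backpack cosmetic mesh, a Fortnite back bling accessory.
# It holds a single body on the pack bone, shaped as one angled capsule offset from the bone origin.
# There are no constraints, so the pack has no jointed simulated parts within this asset.
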PhysicsAsset /BRCosmetics/Accessories/FORT_Backpacks/Backpack_SwissKale/Meshes/M_MED_SwissKale_Pack_PhysicsAsset
bodies {pack [{shapeType=Capsule, length=42.888, radius=5.7587, center=<-1.564 -4.4228 -6.4436>, rotation=<rot -0.8407 -90 -29.877>}]}
constraints []
